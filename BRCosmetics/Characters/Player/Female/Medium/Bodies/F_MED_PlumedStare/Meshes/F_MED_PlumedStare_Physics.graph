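# Physics asset for the PlumedStare female medium body cosmetic, covering dangling hip and right-leg accessories.
# Capsule bodies on pelvis, spine_01 and right thigh root short chains for hip bows, bow string, leg-scroll bow and scroll string.
# Anchor links (base bones, scroll, spine, thigh) have free angular motion.
# Dangling segments use limited swing cones and near-zero twist limits, narrowed on some bow tips, so they sway without spinning.
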PhysicsAsset /BRCosmetics/Characters/Player/Female/Medium/Bodies/F_MED_PlumedStare/Meshes/F_MED_PlumedStare_Physics
bodies {
 base_hipBowString_01 [{shapeType=Capsule, length=0.505, radius=0.505, shapeName=base_hipBowString_01_capsule, rotation=<rot 0 -90 0>}]
 base_hipBow_l_01 [{shapeType=Capsule, length=0.505, radius=0.505, shapeName=base_hipBow_l_01_capsule, rotation=<rot 0 -90 0>}]
 base_hipBow_r_01 [{shapeType=Capsule, length=0.505, radius=0.505, shapeName=base_hipBow_r_01_capsule, rotation=<rot 0 -90 0>}]
 base_legScrollBow_01 [{shapeType=Capsule, length=0.505, radius=0.505, shapeName=base_legScrollBow_01_capsule, rotation=<rot 0 -90 0>}]
 dyn_hipBowString_01 [{shapeType=Capsule, length=4.0414, radius=1.176, shapeName=dyn_hipBowString_01_capsule, center=<2.8369 0.0009 -0.3823>, rotation=<rot 90 0 0>}]
 dyn_hipBowString_02 [{shapeType=Capsule, length=5.2374, radius=1.1317, shapeName=dyn_hipBowString_02_capsule, center=<3.081 0 -0.3832>, rotation=<rot 90 0 0>}]
 dyn_hipBowString_03 [{
   shapeType=Capsule
   length=4.0534
   radius=1.8788
   shapeName=dyn_hipBowString_03_capsule
   center=<3.5142 -0.3426 -0.5832>
   rotation=<rot 84.4407 180 179.9999>
  }]
 dyn_hipBow_l_01 [{shapeType=Capsule, length=2, radius=1, shapeName=dyn_hipBow_l_01_capsule, center=<2 0 0>, rotation=<rot 90 0 0>}]
 dyn_hipBow_l_02 [{shapeType=Capsule, length=2, radius=1, shapeName=dyn_hipBow_l_02_capsule, center=<2 0 0>, rotation=<rot 90 0 0>}]
 dyn_hipBow_r_01 [{shapeType=Capsule, length=2, radius=1, shapeName=dyn_hipBow_r_01_capsule, center=<2 0 0>, rotation=<rot 90 0 0>}]
 dyn_hipBow_r_02 [{shapeType=Capsule, length=2, radius=1, shapeName=dyn_hipBow_r_02_capsule, center=<2 0 0>, rotation=<rot 90 0 0>}]
 dyn_legScrollBow_01 [{shapeType=Capsule, length=2, radius=1, shapeName=dyn_legScrollBow_01_capsule, center=<2 0 0>, rotation=<rot 90 0 0>}]
 legScrollString_01 [{shapeType=Capsule, length=2, radius=1, shapeName=legScrollString_01_capsule, center=<2 0 0>, rotation=<rot 90 0 0>}]
 legScrollString_02 [{shapeType=Capsule, length=3.1943, radius=1.3114, shapeName=legScrollString_02_capsule, center=<2.7041 0 0>, rotation=<rot 90 0 0>}]
 legScroll_02 [{shapeType=Capsule, length=10.7973, radius=2.5623, shapeName=legScroll_02_capsule, rotation=<rot -5.979 3.9837 -33.7654>}]
 pelvis [{shapeType=Capsule, length=11.6077, radius=9.2423, shapeName=pelvis_capsule, center=<0 -2.5794 0>}]
 spine_01 [{shapeType=Capsule, length=0.505, radius=0.505, shapeName=spine_01_capsule, rotation=<rot 0 -90 0>}]
 thigh_r [{shapeType=Capsule, length=18.0363, radius=7.8863, shapeName=thigh_r_capsule, center=<13.7217 0.966 -0.124>, rotation=<rot 90 7.125 18.1575>}]
}
constraints [
 {bone1Name=dyn_hipBowString_01, bone2Name=base_hipBowString_01, swing1LimitDegrees=35, swing2LimitDegrees=35, twistLimitDegrees=1}
 {bone1Name=dyn_hipBowString_02, bone2Name=dyn_hipBowString_01, swing1LimitDegrees=35, swing2LimitDegrees=35, twistLimitDegrees=1}
 {bone1Name=dyn_hipBowString_03, bone2Name=dyn_hipBowString_02, swing1LimitDegrees=35, swing2LimitDegrees=35, twistLimitDegrees=1}
 {bone1Name=dyn_hipBow_r_01, bone2Name=base_hipBow_r_01, swing1LimitDegrees=35, swing2LimitDegrees=30, twistLimitDegrees=1}
 {bone1Name=dyn_hipBow_r_02, bone2Name=dyn_hipBow_r_01, swing1LimitDegrees=25, swing2LimitDegrees=10, twistLimitDegrees=1}
 {
  bone1Name=base_hipBowString_01
  bone2Name=spine_01
  swing1LimitDegrees=45
  swing1Motion=Free
  swing2LimitDegrees=45
  swing2Motion=Free
  twistLimitDegrees=45
  twistMotion=Free
 }
 {
  bone1Name=base_hipBow_r_01
  bone2Name=spine_01
  swing1LimitDegrees=45
  swing1Motion=Free
  swing2LimitDegrees=45
  swing2Motion=Free
  twistLimitDegrees=45
  twistMotion=Free
 }
 {
  bone1Name=spine_01
  bone2Name=pelvis
  swing1LimitDegrees=45
  swing1Motion=Free
  swing2LimitDegrees=45
  swing2Motion=Free
  twistLimitDegrees=45
  twistMotion=Free
 }
 {
  bone1Name=thigh_r
  bone2Name=pelvis
  swing1LimitDegrees=45
  swing1Motion=Free
  swing2LimitDegrees=45
  swing2Motion=Free
  twistLimitDegrees=45
  twistMotion=Free
 }
 {
  bone1Name=legScroll_02
  bone2Name=thigh_r
  swing1LimitDegrees=45
  swing1Motion=Free
  swing2LimitDegrees=45
  swing2Motion=Free
  twistLimitDegrees=45
  twistMotion=Free
 }
 {bone1Name=legScrollString_01, bone2Name=legScroll_02, swing1LimitDegrees=25, swing2LimitDegrees=25, twistLimitDegrees=1}
 {bone1Name=legScrollString_02, bone2Name=legScrollString_01, swing1LimitDegrees=25, swing2LimitDegrees=25, twistLimitDegrees=1}
 {
  bone1Name=base_hipBow_l_01
  bone2Name=spine_01
  swing1LimitDegrees=45
  swing1Motion=Free
  swing2LimitDegrees=45
  swing2Motion=Free
  twistLimitDegrees=45
  twistMotion=Free
 }
 {bone1Name=dyn_hipBow_l_01, bone2Name=base_hipBow_l_01, swing1LimitDegrees=30, swing2LimitDegrees=30, twistLimitDegrees=1}
 {bone1Name=dyn_hipBow_l_02, bone2Name=dyn_hipBow_l_01, swing1LimitDegrees=30, swing2LimitDegrees=10, twistLimitDegrees=1}
 {
  bone1Name=base_legScrollBow_01
  bone2Name=thigh_r
  swing1LimitDegrees=45
  swing1Motion=Free
  swing2LimitDegrees=45
  swing2Motion=Free
  twistLimitDegrees=45
  twistMotion=Free
 }
 {bone1Name=dyn_legScrollBow_01, bone2Name=base_legScrollBow_01, swing1LimitDegrees=25, swing2LimitDegrees=25, twistLimitDegrees=1}
]
