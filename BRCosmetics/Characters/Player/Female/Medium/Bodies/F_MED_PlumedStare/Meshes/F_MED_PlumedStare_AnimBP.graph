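# Animation Blueprint for the PlumedStare female body cosmetic, adding procedural secondary motion to its skirt, feathers and accessories.
# It copies the attached parent's pose through a Control Rig, then applies weighted constraints to hip bow, leg scroll and bullet bones.
# Additive bone transforms adjust skirt bones.
# AnimDynamics and RigidBody nodes simulate the cached correctives pose with overridden gravity, before an emote-driven bow-string offset produces the output.
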AnimBlueprint /BRCosmetics/Characters/Player/Female/Medium/Bodies/F_MED_PlumedStare/Meshes/F_MED_PlumedStare_AnimBP
target_skeleton=&/F_MED_PlumedStare_Skeleton
graphs [
 AnimGraph {nodes [
   Output Pose(Result=Use cached pose 'Game AnimDynamics'()) {layerGroup=None}
   n1 = Control Rig(Source=Copy Pose From Mesh() {bUseAttachedParent=true}) {
    alphaBoolBlend {blendOption=Linear}
    controlRigAssetReference {blueprintRigClass=&/F_MED_PlumedStare_CtrlRig.F_MED_PlumedStare_CtrlRig_C}
   }
   Save cached pose 'Start Pose'(Pose=@n1) {cacheName=Start Pose}
   n2 = "Constraint - hipBowMaster (PPPPP)"!(
    ComponentPose=Local To Component(LocalPose=Use cached pose 'Start Pose'())
    Alpha=1
    ConstraintWeights_0=0.3244
    ConstraintWeights_1:float=0.2554
    ConstraintWeights_2:float=0.1583
    ConstraintWeights_3:float=0.1424
    ConstraintWeights_4:float=0.1194
   ) -> (Pose) {
    alphaBoolBlend {blendOption=Linear}
    boneToModify {boneName=hipBowMaster}
    constraintSetup [
     {targetBone {boneName=pelvis}}
     {targetBone {boneName=spine_02}}
     {targetBone {boneName=spine_01}}
     {targetBone {boneName=spine_04}}
     {targetBone {boneName=spine_03}}
    ]
    constraintWeights [1,1,1,1,1]
   }
   n3 = "Constraint - legScroll_01 (PPPPP)"!(
    ComponentPose=@n2
    Alpha=1
    ConstraintWeights_0=0.6233
    ConstraintWeights_1:float=0.2379
    ConstraintWeights_2:float=0.0951
    ConstraintWeights_3:float=0.0226
    ConstraintWeights_4:float=0.0211
   ) -> (Pose) {
    alphaBoolBlend {blendOption=Linear}
    boneToModify {boneName=legScroll_01}
    constraintSetup [
     {targetBone {boneName=thigh_twist_01_r}}
     {targetBone {boneName=thigh_r}}
     {targetBone {boneName=pelvis}}
     {targetBone {boneName=deform_groin_r}}
     {targetBone {boneName=deform_glute_r}}
    ]
    constraintWeights [1,1,1,1,1]
   }
   n4 = "Constraint - base_legScroll_02 (PPPPP)"!(
    ComponentPose=@n3
    Alpha=1
    ConstraintWeights_0=0.5853
    ConstraintWeights_1:float=0.3265
    ConstraintWeights_2:float=0.0575
    ConstraintWeights_3:float=0.0251
    ConstraintWeights_4:float=0.0055
   ) -> (Pose) {
    alphaBoolBlend {blendOption=Linear}
    boneToModify {boneName=base_legScroll_02}
    constraintSetup [
     {targetBone {boneName=thigh_twist_01_r}}
     {targetBone {boneName=thigh_r}}
     {targetBone {boneName=pelvis}}
     {targetBone {boneName=deform_glute_r}}
     {targetBone {boneName=deform_groin_r}}
    ]
    constraintWeights [1,1,1,1,1]
   }
   n5 = "Constraint - base_bulletFront_01 (PPPPP)"!(
    ComponentPose=@n4
    Alpha=1
    ConstraintWeights_0=0.5094
    ConstraintWeights_1:float=0.2353
    ConstraintWeights_2:float=0.2131
    ConstraintWeights_3:float=0.0245
    ConstraintWeights_4:float=0.0177
   ) -> (Pose) {
    alphaBoolBlend {blendOption=Linear}
    boneToModify {boneName=base_bulletFront_01}
    constraintSetup [
     {targetBone {boneName=pelvis}}
     {targetBone {boneName=spine_01}}
     {targetBone {boneName=spine_02}}
     {targetBone {boneName=spine_04}}
     {targetBone {boneName=spine_03}}
    ]
    constraintWeights [1,1,1,1,1]
   }
   n6 = "Constraint - base_bulletFront_02 (PPPPP)"!(
    ComponentPose=@n5
    Alpha=1
    ConstraintWeights_0=0.5288
    ConstraintWeights_1:float=0.2241
    ConstraintWeights_2:float=0.2062
    ConstraintWeights_3:float=0.0238
    ConstraintWeights_4:float=0.0172
   ) -> (Pose) {
    alphaBoolBlend {blendOption=Linear}
    boneToModify {boneName=base_bulletFront_02}
    constraintSetup [
     {targetBone {boneName=pelvis}}
     {targetBone {boneName=spine_01}}
     {targetBone {boneName=spine_02}}
     {targetBone {boneName=spine_04}}
     {targetBone {boneName=spine_03}}
    ]
    constraintWeights [1,1,1,1,1]
   }
   n7 = "Constraint - base_bulletFront_03 (PPPPP)"!(
    ComponentPose=@n6
    Alpha=1
    ConstraintWeights_0=0.519
    ConstraintWeights_1:float=0.2325
    ConstraintWeights_2:float=0.2076
    ConstraintWeights_3:float=0.0238
    ConstraintWeights_4:float=0.0171
   ) -> (Pose) {
    alphaBoolBlend {blendOption=Linear}
    boneToModify {boneName=base_bulletFront_03}
    constraintSetup [
     {targetBone {boneName=pelvis}}
     {targetBone {boneName=spine_01}}
     {targetBone {boneName=spine_02}}
     {targetBone {boneName=spine_04}}
     {targetBone {boneName=spine_03}}
    ]
    constraintWeights [1,1,1,1,1]
   }
   n8 = "Constraint - base_bulletFront_04 (PPPPP)"!(
    ComponentPose=@n7
    Alpha=1
    ConstraintWeights_0=0.515
    ConstraintWeights_1:float=0.2415
    ConstraintWeights_2:float=0.2039
    ConstraintWeights_3:float=0.0231
    ConstraintWeights_4:float=0.0165
   ) -> (Pose) {
    alphaBoolBlend {blendOption=Linear}
    boneToModify {boneName=base_bulletFront_04}
    constraintSetup [
     {targetBone {boneName=pelvis}}
     {targetBone {boneName=spine_01}}
     {targetBone {boneName=spine_02}}
     {targetBone {boneName=spine_04}}
     {targetBone {boneName=spine_03}}
    ]
    constraintWeights [1,1,1,1,1]
   }
   n9 = "Constraint - base_bulletBack_01 (PPPPP)"!(
    ComponentPose=@n8
    Alpha=1
    ConstraintWeights_0=0.5031
    ConstraintWeights_1:float=0.2338
    ConstraintWeights_2:float=0.2186
    ConstraintWeights_3:float=0.0257
    ConstraintWeights_4:float=0.0187
   ) -> (Pose) {
    alphaBoolBlend {blendOption=Linear}
    boneToModify {boneName=base_bulletBack_01}
    constraintSetup [
     {targetBone {boneName=pelvis}}
     {targetBone {boneName=spine_01}}
     {targetBone {boneName=spine_02}}
     {targetBone {boneName=spine_04}}
     {targetBone {boneName=spine_03}}
    ]
    constraintWeights [1,1,1,1,1]
   }
   n10 = "Constraint - base_bulletBack_02 (PPPPP)"!(
    ComponentPose=@n9
    Alpha=1
    ConstraintWeights_0=0.5078
    ConstraintWeights_1:float=0.2335
    ConstraintWeights_2:float=0.2159
    ConstraintWeights_3:float=0.0249
    ConstraintWeights_4:float=0.018
   ) -> (Pose) {
    alphaBoolBlend {blendOption=Linear}
    boneToModify {boneName=base_bulletBack_02}
    constraintSetup [
     {targetBone {boneName=pelvis}}
     {targetBone {boneName=spine_01}}
     {targetBone {boneName=spine_02}}
     {targetBone {boneName=spine_04}}
     {targetBone {boneName=spine_03}}
    ]
    constraintWeights [1,1,1,1,1]
   }
   n11 = "Constraint - base_bulletBack_03 (PPPPP)"!(
    ComponentPose=@n10
    Alpha=1
    ConstraintWeights_0=0.5078
    ConstraintWeights_1:float=0.2335
    ConstraintWeights_2:float=0.2159
    ConstraintWeights_3:float=0.0249
    ConstraintWeights_4:float=0.018
   ) -> (Pose) {
    alphaBoolBlend {blendOption=Linear}
    boneToModify {boneName=base_bulletBack_03}
    constraintSetup [
     {targetBone {boneName=pelvis}}
     {targetBone {boneName=spine_01}}
     {targetBone {boneName=spine_02}}
     {targetBone {boneName=spine_04}}
     {targetBone {boneName=spine_03}}
    ]
    constraintWeights [1,1,1,1,1]
   }
   Save cached pose 'Surface Constraints'(Pose=Component To Local(ComponentPose=@n11)) {cacheName=Surface Constraints}
   n12 = Get GravityOverride()
   n13 = Anim Dynamics - Bone: dyn_skirtSide_01!(
    SimulationSpace:AnimPhysSimSpaceType Enum
    ComponentPose=Local To Component(LocalPose=Use cached pose 'Correctives'())
    Alpha=1
    GravityOverride:Vector=@n12
    ComponentLinearAccScale:Vector="0.5,0.5,0.5"
    ComponentLinearVelScale:Vector="0.5,0.5,0.5"
    ExternalForce="0,0,0"
   ) -> (Pose) {
    angularDampingOverride=1
    angularSpringConstant=10000
    bAngularSpring=true
    bOverrideAngularDamping=true
    bUseGravityOverride=true
    gravityScale=0.35
    simulationSpace=RootRelative
    alphaBoolBlend {blendOption=Linear}
    boundBone {boneName=dyn_skirtSide_01}
    componentAppliedLinearAccClamp=<5000 5000 5000>
    componentLinearAccScale=<0.5 0.5 0.5>
    componentLinearVelScale=<0.5 0.5 0.5>
    gravityOverride {z=-980}
    physicsBodyDefinitions [{
      boundBone {boneName=dyn_skirtSide_01}
      boxExtents=<30 30 30>
      constraintSetup {angularLimitsMax=<0 5 0>, angularLimitsMin=<0 -5 -15>, angularTarget=<1 0 0>}
      localJointOffset=<4 0 0>
     }]
    relativeSpaceBone {boneName=pelvis}
   }
   n14 = Anim Dynamics - Bone: dyn_skirtSide_02!(
    SimulationSpace:AnimPhysSimSpaceType Enum
    ComponentPose=@n13
    Alpha=1
    GravityOverride:Vector=@n12
    ComponentLinearAccScale:Vector="0.5,0.5,0.5"
    ComponentLinearVelScale:Vector="0.5,0.5,0.5"
    ExternalForce="0,0,0"
   ) -> (Pose) {
    angularDampingOverride=1
    angularSpringConstant=10000
    bAngularSpring=true
    bOverrideAngularDamping=true
    bUseGravityOverride=true
    gravityScale=0.35
    simulationSpace=RootRelative
    alphaBoolBlend {blendOption=Linear}
    boundBone {boneName=dyn_skirtSide_02}
    componentAppliedLinearAccClamp=<5000 5000 5000>
    componentLinearAccScale=<0.5 0.5 0.5>
    componentLinearVelScale=<0.5 0.5 0.5>
    gravityOverride {z=-980}
    physicsBodyDefinitions [{
      boundBone {boneName=dyn_skirtSide_02}
      boxExtents=<30 30 30>
      constraintSetup {angularLimitsMax=<0 5 0>, angularLimitsMin=<0 -5 -15>, angularTarget=<1 0 0>}
      localJointOffset=<4 0 0>
     }]
    relativeSpaceBone {boneName=pelvis}
   }
   n15 = Anim Dynamics - Bone: dyn_skirtSide_03!(
    SimulationSpace:AnimPhysSimSpaceType Enum
    ComponentPose=@n14
    Alpha=1
    GravityOverride:Vector=@n12
    ComponentLinearAccScale:Vector="0.5,0.5,0.5"
    ComponentLinearVelScale:Vector="0.5,0.5,0.5"
    ExternalForce="0,0,0"
   ) -> (Pose) {
    angularDampingOverride=1
    angularSpringConstant=10000
    bAngularSpring=true
    bOverrideAngularDamping=true
    bUseGravityOverride=true
    gravityScale=0.35
    simulationSpace=RootRelative
    alphaBoolBlend {blendOption=Linear}
    boundBone {boneName=dyn_skirtSide_03}
    componentAppliedLinearAccClamp=<5000 5000 5000>
    componentLinearAccScale=<0.5 0.5 0.5>
    componentLinearVelScale=<0.5 0.5 0.5>
    gravityOverride {z=-980}
    physicsBodyDefinitions [{
      boundBone {boneName=dyn_skirtSide_03}
      boxExtents=<30 30 30>
      constraintSetup {angularLimitsMax=<0 5 5>, angularLimitsMin=<0 -5 -15>, angularTarget=<1 0 0>}
      localJointOffset=<4 0 0>
     }]
    relativeSpaceBone {boneName=pelvis}
   }
   n16 = Anim Dynamics - Bone: dyn_skirtBackSide_02!(
    SimulationSpace:AnimPhysSimSpaceType Enum
    ComponentPose=@n15
    Alpha=1
    GravityOverride:Vector=@n12
    ComponentLinearAccScale:Vector="0.5,0.5,0.5"
    ComponentLinearVelScale:Vector="0.5,0.5,0.5"
    ExternalForce="0,0,0"
   ) -> (Pose) {
    angularDampingOverride=1
    angularSpringConstant=10000
    bAngularSpring=true
    bOverrideAngularDamping=true
    bUseGravityOverride=true
    gravityScale=0.35
    simulationSpace=RootRelative
    alphaBoolBlend {blendOption=Linear}
    boundBone {boneName=dyn_skirtBackSide_02}
    componentAppliedLinearAccClamp=<5000 5000 5000>
    componentLinearAccScale=<0.5 0.5 0.5>
    componentLinearVelScale=<0.5 0.5 0.5>
    gravityOverride {z=-980}
    physicsBodyDefinitions [{
      boundBone {boneName=dyn_skirtBackSide_02}
      boxExtents=<30 30 30>
      constraintSetup {angularLimitsMax=<0 5 0>, angularLimitsMin=<0 -5 -15>, angularTarget=<1 0 0>}
      localJointOffset=<4 0 0>
     }]
    relativeSpaceBone {boneName=pelvis}
   }
   n17 = Anim Dynamics - Bone: dyn_skirtBackSide_03!(
    SimulationSpace:AnimPhysSimSpaceType Enum
    ComponentPose=@n16
    Alpha=1
    GravityOverride:Vector=@n12
    ComponentLinearAccScale:Vector="0.5,0.5,0.5"
    ComponentLinearVelScale:Vector="0.5,0.5,0.5"
    ExternalForce="0,0,0"
   ) -> (Pose) {
    angularDampingOverride=1
    angularSpringConstant=10000
    bAngularSpring=true
    bOverrideAngularDamping=true
    bUseGravityOverride=true
    gravityScale=0.35
    simulationSpace=RootRelative
    alphaBoolBlend {blendOption=Linear}
    boundBone {boneName=dyn_skirtBackSide_03}
    componentAppliedLinearAccClamp=<5000 5000 5000>
    componentLinearAccScale=<0.5 0.5 0.5>
    componentLinearVelScale=<0.5 0.5 0.5>
    gravityOverride {z=-980}
    physicsBodyDefinitions [{
      boundBone {boneName=dyn_skirtBackSide_03}
      boxExtents=<30 30 30>
      constraintSetup {angularLimitsMax=<0 5 5>, angularLimitsMin=<0 -5 -15>, angularTarget=<1 0 0>}
      localJointOffset=<4 0 0>
     }]
    relativeSpaceBone {boneName=pelvis}
   }
   n18 = Anim Dynamics - Bone: dyn_skirtBackSide_04!(
    SimulationSpace:AnimPhysSimSpaceType Enum
    ComponentPose=@n17
    Alpha=1
    GravityOverride:Vector=@n12
    ComponentLinearAccScale:Vector="0.5,0.5,0.5"
    ComponentLinearVelScale:Vector="0.5,0.5,0.5"
    ExternalForce="0,0,0"
   ) -> (Pose) {
    angularDampingOverride=1
    angularSpringConstant=10000
    bAngularSpring=true
    bOverrideAngularDamping=true
    bUseGravityOverride=true
    gravityScale=0.35
    simulationSpace=RootRelative
    alphaBoolBlend {blendOption=Linear}
    boundBone {boneName=dyn_skirtBackSide_04}
    componentAppliedLinearAccClamp=<5000 5000 5000>
    componentLinearAccScale=<0.5 0.5 0.5>
    componentLinearVelScale=<0.5 0.5 0.5>
    gravityOverride {z=-980}
    physicsBodyDefinitions [{
      boundBone {boneName=dyn_skirtBackSide_04}
      boxExtents=<30 30 30>
      constraintSetup {angularLimitsMax=<0 5 15>, angularLimitsMin=<0 -5 -15>, angularTarget=<1 0 0>}
      localJointOffset=<4 0 0>
     }]
    relativeSpaceBone {boneName=pelvis}
   }
   n19 = Anim Dynamics - Bone: dyn_skirtBack_01!(
    SimulationSpace:AnimPhysSimSpaceType Enum
    ComponentPose=@n18
    Alpha=1
    GravityOverride:Vector=@n12
    ComponentLinearAccScale:Vector="0.5,0.5,0.5"
    ComponentLinearVelScale:Vector="0.5,0.5,0.5"
    ExternalForce="0,0,0"
   ) -> (Pose) {
    angularDampingOverride=1
    angularSpringConstant=10000
    bAngularSpring=true
    bOverrideAngularDamping=true
    bUseGravityOverride=true
    gravityScale=0.35
    simulationSpace=RootRelative
    alphaBoolBlend {blendOption=Linear}
    boundBone {boneName=dyn_skirtBack_01}
    componentAppliedLinearAccClamp=<5000 5000 5000>
    componentLinearAccScale=<0.5 0.5 0.5>
    componentLinearVelScale=<0.5 0.5 0.5>
    gravityOverride {z=-980}
    physicsBodyDefinitions [{
      boundBone {boneName=dyn_skirtBack_01}
      boxExtents=<30 30 30>
      constraintSetup {angularLimitsMax=<0 5 0>, angularLimitsMin=<0 -5 -15>, angularTarget=<1 0 0>}
      localJointOffset=<4 0 0>
     }]
    relativeSpaceBone {boneName=pelvis}
   }
   n20 = Anim Dynamics - Bone: dyn_skirtBack_02!(
    SimulationSpace:AnimPhysSimSpaceType Enum
    ComponentPose=@n19
    Alpha=1
    GravityOverride:Vector=@n12
    ComponentLinearAccScale:Vector="0.5,0.5,0.5"
    ComponentLinearVelScale:Vector="0.5,0.5,0.5"
    ExternalForce="0,0,0"
   ) -> (Pose) {
    angularDampingOverride=1
    angularSpringConstant=10000
    bAngularSpring=true
    bOverrideAngularDamping=true
    bUseGravityOverride=true
    gravityScale=0.35
    simulationSpace=RootRelative
    alphaBoolBlend {blendOption=Linear}
    boundBone {boneName=dyn_skirtBack_02}
    componentAppliedLinearAccClamp=<5000 5000 5000>
    componentLinearAccScale=<0.5 0.5 0.5>
    componentLinearVelScale=<0.5 0.5 0.5>
    gravityOverride {z=-980}
    physicsBodyDefinitions [{
      boundBone {boneName=dyn_skirtBack_02}
      boxExtents=<30 30 30>
      constraintSetup {angularLimitsMax=<0 5 0>, angularLimitsMin=<0 -5 -15>, angularTarget=<1 0 0>}
      localJointOffset=<4 0 0>
     }]
    relativeSpaceBone {boneName=pelvis}
   }
   n21 = Anim Dynamics - Bone: dyn_skirtBackSide_01!(
    SimulationSpace:AnimPhysSimSpaceType Enum
    ComponentPose=@n20
    Alpha=1
    GravityOverride:Vector=@n12
    ComponentLinearAccScale:Vector="0.5,0.5,0.5"
    ComponentLinearVelScale:Vector="0.5,0.5,0.5"
    ExternalForce="0,0,0"
   ) -> (Pose) {
    angularDampingOverride=1
    angularSpringConstant=10000
    bAngularSpring=true
    bOverrideAngularDamping=true
    bUseGravityOverride=true
    gravityScale=0.35
    simulationSpace=RootRelative
    alphaBoolBlend {blendOption=Linear}
    boundBone {boneName=dyn_skirtBackSide_01}
    componentAppliedLinearAccClamp=<5000 5000 5000>
    componentLinearAccScale=<0.5 0.5 0.5>
    componentLinearVelScale=<0.5 0.5 0.5>
    gravityOverride {z=-980}
    physicsBodyDefinitions [{
      boundBone {boneName=dyn_skirtBackSide_01}
      boxExtents=<30 30 30>
      constraintSetup {angularLimitsMax=<0 5 0>, angularLimitsMin=<0 -5 -15>, angularTarget=<1 0 0>}
      localJointOffset=<4 0 0>
     }]
    relativeSpaceBone {boneName=pelvis}
   }
   n22 = Anim Dynamics - Bone: dyn_skirtFeather_01_01!(
    SimulationSpace:AnimPhysSimSpaceType Enum
    ComponentPose=@n21
    Alpha=1
    GravityOverride:Vector=@n12
    ComponentLinearAccScale:Vector="0.5,0.5,0.5"
    ComponentLinearVelScale:Vector="0.5,0.5,0.5"
    ExternalForce="0,0,0"
   ) -> (Pose) {
    angularDampingOverride=1
    angularSpringConstant=10000
    bAngularSpring=true
    bOverrideAngularDamping=true
    bUseGravityOverride=true
    gravityScale=0.35
    simulationSpace=RootRelative
    alphaBoolBlend {blendOption=Linear}
    boundBone {boneName=dyn_skirtFeather_01_01}
    componentAppliedLinearAccClamp=<5000 5000 5000>
    componentLinearAccScale=<0.5 0.5 0.5>
    componentLinearVelScale=<0.5 0.5 0.5>
    gravityOverride {z=-980}
    physicsBodyDefinitions [{
      boundBone {boneName=dyn_skirtFeather_01_01}
      boxExtents=<30 30 30>
      constraintSetup {angularLimitsMax=<0 5 15>, angularLimitsMin=<0 -5 -15>, angularTarget=<1 0 0>}
      localJointOffset=<4 0 0>
     }]
    relativeSpaceBone {boneName=pelvis}
   }
   n23 = Anim Dynamics - Bone: dyn_skirtFeather_01_02!(
    SimulationSpace:AnimPhysSimSpaceType Enum
    ComponentPose=@n22
    Alpha=1
    GravityOverride:Vector=@n12
    ComponentLinearAccScale:Vector="0.5,0.5,0.5"
    ComponentLinearVelScale:Vector="0.5,0.5,0.5"
    ExternalForce="0,0,0"
   ) -> (Pose) {
    angularDampingOverride=1
    angularSpringConstant=10000
    bAngularSpring=true
    bOverrideAngularDamping=true
    bUseGravityOverride=true
    gravityScale=0.35
    simulationSpace=RootRelative
    alphaBoolBlend {blendOption=Linear}
    boundBone {boneName=dyn_skirtFeather_01_02}
    componentAppliedLinearAccClamp=<5000 5000 5000>
    componentLinearAccScale=<0.5 0.5 0.5>
    componentLinearVelScale=<0.5 0.5 0.5>
    gravityOverride {z=-980}
    physicsBodyDefinitions [{
      boundBone {boneName=dyn_skirtFeather_01_02}
      boxExtents=<30 30 30>
      constraintSetup {angularLimitsMax=<0 5 15>, angularLimitsMin=<0 -5 -15>, angularTarget=<1 0 0>}
      localJointOffset=<4 0 0>
     }]
    relativeSpaceBone {boneName=pelvis}
   }
   n24 = Anim Dynamics - Bone: dyn_skirtFeather_01_03!(
    SimulationSpace:AnimPhysSimSpaceType Enum
    ComponentPose=@n23
    Alpha=1
    GravityOverride:Vector=@n12
    ComponentLinearAccScale:Vector="0.5,0.5,0.5"
    ComponentLinearVelScale:Vector="0.5,0.5,0.5"
    ExternalForce="0,0,0"
   ) -> (Pose) {
    angularDampingOverride=1
    angularSpringConstant=10000
    bAngularSpring=true
    bOverrideAngularDamping=true
    bUseGravityOverride=true
    gravityScale=0.35
    simulationSpace=RootRelative
    alphaBoolBlend {blendOption=Linear}
    boundBone {boneName=dyn_skirtFeather_01_03}
    componentAppliedLinearAccClamp=<5000 5000 5000>
    componentLinearAccScale=<0.5 0.5 0.5>
    componentLinearVelScale=<0.5 0.5 0.5>
    gravityOverride {z=-980}
    physicsBodyDefinitions [{
      boundBone {boneName=dyn_skirtFeather_01_03}
      boxExtents=<30 30 30>
      constraintSetup {angularLimitsMax=<0 5 15>, angularLimitsMin=<0 -5 -15>, angularTarget=<1 0 0>}
      localJointOffset=<4 0 0>
     }]
    relativeSpaceBone {boneName=pelvis}
   }
   n25 = Anim Dynamics - Bone: dyn_skirtFeather_02_01!(
    SimulationSpace:AnimPhysSimSpaceType Enum
    ComponentPose=@n24
    Alpha=1
    GravityOverride:Vector=@n12
    ComponentLinearAccScale:Vector="0.5,0.5,0.5"
    ComponentLinearVelScale:Vector="0.5,0.5,0.5"
    ExternalForce="0,0,0"
   ) -> (Pose) {
    angularDampingOverride=1
    angularSpringConstant=10000
    bAngularSpring=true
    bOverrideAngularDamping=true
    bUseGravityOverride=true
    gravityScale=0.35
    simulationSpace=RootRelative
    alphaBoolBlend {blendOption=Linear}
    boundBone {boneName=dyn_skirtFeather_02_01}
    componentAppliedLinearAccClamp=<5000 5000 5000>
    componentLinearAccScale=<0.5 0.5 0.5>
    componentLinearVelScale=<0.5 0.5 0.5>
    gravityOverride {z=-980}
    physicsBodyDefinitions [{
      boundBone {boneName=dyn_skirtFeather_02_01}
      boxExtents=<30 30 30>
      constraintSetup {angularLimitsMax=<0 5 15>, angularLimitsMin=<0 -5 -15>, angularTarget=<1 0 0>}
      localJointOffset=<4 0 0>
     }]
    relativeSpaceBone {boneName=pelvis}
   }
   n26 = Anim Dynamics - Bone: dyn_skirtFeather_02_02!(
    SimulationSpace:AnimPhysSimSpaceType Enum
    ComponentPose=@n25
    Alpha=1
    GravityOverride:Vector=@n12
    ComponentLinearAccScale:Vector="0.5,0.5,0.5"
    ComponentLinearVelScale:Vector="0.5,0.5,0.5"
    ExternalForce="0,0,0"
   ) -> (Pose) {
    angularDampingOverride=1
    angularSpringConstant=10000
    bAngularSpring=true
    bOverrideAngularDamping=true
    bUseGravityOverride=true
    gravityScale=0.35
    simulationSpace=RootRelative
    alphaBoolBlend {blendOption=Linear}
    boundBone {boneName=dyn_skirtFeather_02_02}
    componentAppliedLinearAccClamp=<5000 5000 5000>
    componentLinearAccScale=<0.5 0.5 0.5>
    componentLinearVelScale=<0.5 0.5 0.5>
    gravityOverride {z=-980}
    physicsBodyDefinitions [{
      boundBone {boneName=dyn_skirtFeather_02_02}
      boxExtents=<30 30 30>
      constraintSetup {angularLimitsMax=<0 5 15>, angularLimitsMin=<0 -5 -15>, angularTarget=<1 0 0>}
      localJointOffset=<4 0 0>
     }]
    relativeSpaceBone {boneName=pelvis}
   }
   n27 = Anim Dynamics - Bone: dyn_skirtFeather_02_03!(
    SimulationSpace:AnimPhysSimSpaceType Enum
    ComponentPose=@n26
    Alpha=1
    GravityOverride:Vector=@n12
    ComponentLinearAccScale:Vector="0.5,0.5,0.5"
    ComponentLinearVelScale:Vector="0.5,0.5,0.5"
    ExternalForce="0,0,0"
   ) -> (Pose) {
    angularDampingOverride=1
    angularSpringConstant=10000
    bAngularSpring=true
    bOverrideAngularDamping=true
    bUseGravityOverride=true
    gravityScale=0.35
    simulationSpace=RootRelative
    alphaBoolBlend {blendOption=Linear}
    boundBone {boneName=dyn_skirtFeather_02_03}
    componentAppliedLinearAccClamp=<5000 5000 5000>
    componentLinearAccScale=<0.5 0.5 0.5>
    componentLinearVelScale=<0.5 0.5 0.5>
    gravityOverride {z=-980}
    physicsBodyDefinitions [{
      boundBone {boneName=dyn_skirtFeather_02_03}
      boxExtents=<30 30 30>
      constraintSetup {angularLimitsMax=<0 5 15>, angularLimitsMin=<0 -5 -15>, angularTarget=<1 0 0>}
      localJointOffset=<4 0 0>
     }]
    relativeSpaceBone {boneName=pelvis}
   }
   n28 = Anim Dynamics - Bone: dyn_skirtFeather_04_01!(
    SimulationSpace:AnimPhysSimSpaceType Enum
    ComponentPose=@n27
    Alpha=1
    GravityOverride:Vector=@n12
    ComponentLinearAccScale:Vector="0.5,0.5,0.5"
    ComponentLinearVelScale:Vector="0.5,0.5,0.5"
    ExternalForce="0,0,0"
   ) -> (Pose) {
    angularDampingOverride=1
    angularSpringConstant=10000
    bAngularSpring=true
    bOverrideAngularDamping=true
    bUseGravityOverride=true
    gravityScale=0.35
    simulationSpace=RootRelative
    alphaBoolBlend {blendOption=Linear}
    boundBone {boneName=dyn_skirtFeather_04_01}
    componentAppliedLinearAccClamp=<5000 5000 5000>
    componentLinearAccScale=<0.5 0.5 0.5>
    componentLinearVelScale=<0.5 0.5 0.5>
    gravityOverride {z=-980}
    physicsBodyDefinitions [{
      boundBone {boneName=dyn_skirtFeather_04_01}
      boxExtents=<30 30 30>
      constraintSetup {angularLimitsMax=<0 5 15>, angularLimitsMin=<0 -5 -15>, angularTarget=<1 0 0>}
      localJointOffset=<4 0 0>
     }]
    relativeSpaceBone {boneName=pelvis}
   }
   n29 = Anim Dynamics - Bone: dyn_skirtFeather_04_02!(
    SimulationSpace:AnimPhysSimSpaceType Enum
    ComponentPose=@n28
    Alpha=1
    GravityOverride:Vector=@n12
    ComponentLinearAccScale:Vector="0.5,0.5,0.5"
    ComponentLinearVelScale:Vector="0.5,0.5,0.5"
    ExternalForce="0,0,0"
   ) -> (Pose) {
    angularDampingOverride=1
    angularSpringConstant=10000
    bAngularSpring=true
    bOverrideAngularDamping=true
    bUseGravityOverride=true
    gravityScale=0.35
    simulationSpace=RootRelative
    alphaBoolBlend {blendOption=Linear}
    boundBone {boneName=dyn_skirtFeather_04_02}
    componentAppliedLinearAccClamp=<5000 5000 5000>
    componentLinearAccScale=<0.5 0.5 0.5>
    componentLinearVelScale=<0.5 0.5 0.5>
    gravityOverride {z=-980}
    physicsBodyDefinitions [{
      boundBone {boneName=dyn_skirtFeather_04_02}
      boxExtents=<30 30 30>
      constraintSetup {angularLimitsMax=<0 5 15>, angularLimitsMin=<0 -5 -15>, angularTarget=<1 0 0>}
      localJointOffset=<4 0 0>
     }]
    relativeSpaceBone {boneName=pelvis}
   }
   n30 = Anim Dynamics - Bone: dyn_skirtFeather_00_01!(
    SimulationSpace:AnimPhysSimSpaceType Enum
    ComponentPose=@n29
    Alpha=1
    GravityOverride:Vector=@n12
    ComponentLinearAccScale:Vector="0.5,0.5,0.5"
    ComponentLinearVelScale:Vector="0.5,0.5,0.5"
    ExternalForce="0,0,0"
   ) -> (Pose) {
    angularDampingOverride=1
    angularSpringConstant=10000
    bAngularSpring=true
    bOverrideAngularDamping=true
    bUseGravityOverride=true
    gravityScale=0.35
    simulationSpace=RootRelative
    alphaBoolBlend {blendOption=Linear}
    boundBone {boneName=dyn_skirtFeather_00_01}
    componentAppliedLinearAccClamp=<5000 5000 5000>
    componentLinearAccScale=<0.5 0.5 0.5>
    componentLinearVelScale=<0.5 0.5 0.5>
    gravityOverride {z=-980}
    physicsBodyDefinitions [{
      boundBone {boneName=dyn_skirtFeather_00_01}
      boxExtents=<30 30 30>
      constraintSetup {angularLimitsMax=<0 5 15>, angularLimitsMin=<0 -5 -15>, angularTarget=<1 0 0>}
      localJointOffset=<4 0 0>
     }]
    relativeSpaceBone {boneName=pelvis}
   }
   n31 = Anim Dynamics - Bone: dyn_skirtFeather_03_01!(
    SimulationSpace:AnimPhysSimSpaceType Enum
    ComponentPose=@n30
    Alpha=1
    GravityOverride:Vector=@n12
    ComponentLinearAccScale:Vector="0.5,0.5,0.5"
    ComponentLinearVelScale:Vector="0.5,0.5,0.5"
    ExternalForce="0,0,0"
   ) -> (Pose) {
    angularDampingOverride=1
    angularSpringConstant=10000
    bAngularSpring=true
    bOverrideAngularDamping=true
    bUseGravityOverride=true
    gravityScale=0.35
    simulationSpace=RootRelative
    alphaBoolBlend {blendOption=Linear}
    boundBone {boneName=dyn_skirtFeather_03_01}
    componentAppliedLinearAccClamp=<5000 5000 5000>
    componentLinearAccScale=<0.5 0.5 0.5>
    componentLinearVelScale=<0.5 0.5 0.5>
    gravityOverride {z=-980}
    physicsBodyDefinitions [{
      boundBone {boneName=dyn_skirtFeather_03_01}
      boxExtents=<30 30 30>
      constraintSetup {angularLimitsMax=<0 5 15>, angularLimitsMin=<0 -5 -15>, angularTarget=<1 0 0>}
      localJointOffset=<4 0 0>
     }]
    relativeSpaceBone {boneName=pelvis}
   }
   n32 = Anim Dynamics - Bone: dyn_skirtFeather_03_02!(
    SimulationSpace:AnimPhysSimSpaceType Enum
    ComponentPose=@n31
    Alpha=1
    GravityOverride:Vector=@n12
    ComponentLinearAccScale:Vector="0.5,0.5,0.5"
    ComponentLinearVelScale:Vector="0.5,0.5,0.5"
    ExternalForce="0,0,0"
   ) -> (Pose) {
    angularDampingOverride=1
    angularSpringConstant=10000
    bAngularSpring=true
    bOverrideAngularDamping=true
    bUseGravityOverride=true
    gravityScale=0.35
    simulationSpace=RootRelative
    alphaBoolBlend {blendOption=Linear}
    boundBone {boneName=dyn_skirtFeather_03_02}
    componentAppliedLinearAccClamp=<5000 5000 5000>
    componentLinearAccScale=<0.5 0.5 0.5>
    componentLinearVelScale=<0.5 0.5 0.5>
    gravityOverride {z=-980}
    physicsBodyDefinitions [{
      boundBone {boneName=dyn_skirtFeather_03_02}
      boxExtents=<30 30 30>
      constraintSetup {angularLimitsMax=<0 5 15>, angularLimitsMin=<0 -5 -15>, angularTarget=<1 0 0>}
      localJointOffset=<4 0 0>
     }]
    relativeSpaceBone {boneName=pelvis}
   }
   n33 = Anim Dynamics - Bone: dyn_skirtFeather_03_03!(
    SimulationSpace:AnimPhysSimSpaceType Enum
    ComponentPose=@n32
    Alpha=1
    GravityOverride:Vector=@n12
    ComponentLinearAccScale:Vector="0.5,0.5,0.5"
    ComponentLinearVelScale:Vector="0.5,0.5,0.5"
    ExternalForce="0,0,0"
   ) -> (Pose) {
    angularDampingOverride=1
    angularSpringConstant=10000
    bAngularSpring=true
    bOverrideAngularDamping=true
    bUseGravityOverride=true
    gravityScale=0.35
    simulationSpace=RootRelative
    alphaBoolBlend {blendOption=Linear}
    boundBone {boneName=dyn_skirtFeather_03_03}
    componentAppliedLinearAccClamp=<5000 5000 5000>
    componentLinearAccScale=<0.5 0.5 0.5>
    componentLinearVelScale=<0.5 0.5 0.5>
    gravityOverride {z=-980}
    physicsBodyDefinitions [{
      boundBone {boneName=dyn_skirtFeather_03_03}
      boxExtents=<30 30 30>
      constraintSetup {angularLimitsMax=<0 5 15>, angularLimitsMin=<0 -5 -15>, angularTarget=<1 0 0>}
      localJointOffset=<4 0 0>
     }]
    relativeSpaceBone {boneName=pelvis}
   }
   n34 = Anim Dynamics - Bone: dyn_shoulderFeathers_03!(
    SimulationSpace:AnimPhysSimSpaceType Enum=Component
    ComponentPose=@n33
    Alpha=1
    ComponentLinearAccScale:Vector="1,1,1"
    ComponentLinearVelScale:Vector="1,1,1"
    ExternalForce="0,0,0"
   ) -> (Pose) {
    angularDampingOverride=1
    angularSpringConstant=17000
    bAngularSpring=true
    bOverrideAngularDamping=true
    bUseGravityOverride=true
    gravityScale=0.35
    alphaBoolBlend {blendOption=Linear}
    boundBone {boneName=dyn_shoulderFeathers_03}
    componentAppliedLinearAccClamp=<1000 1000 1000>
    componentLinearAccScale=<1 1 1>
    componentLinearVelScale=<1 1 1>
    gravityOverride {z=-980}
    physicsBodyDefinitions [{
      boundBone {boneName=dyn_shoulderFeathers_03}
      boxExtents=<30 30 30>
      constraintSetup {angularLimitsMax=<10 25 40>, angularLimitsMin=<-10 -25 -40>, angularTarget=<1 0 0>}
      localJointOffset=<5 0 0>
     }]
    relativeSpaceBone {boneName=spine_04}
   }
   n35 = Anim Dynamics - Bone: dyn_shoulderFeathers_01!(
    SimulationSpace:AnimPhysSimSpaceType Enum=Component
    ComponentPose=@n34
    Alpha=1
    ComponentLinearAccScale:Vector="1,1,1"
    ComponentLinearVelScale:Vector="1,1,1"
    ExternalForce="0,0,0"
   ) -> (Pose) {
    angularDampingOverride=1
    angularSpringConstant=17000
    bAngularSpring=true
    bOverrideAngularDamping=true
    bUseGravityOverride=true
    gravityScale=0.35
    alphaBoolBlend {blendOption=Linear}
    boundBone {boneName=dyn_shoulderFeathers_01}
    componentLinearAccScale=<1 1 1>
    componentLinearVelScale=<1 1 1>
    gravityOverride {z=-980}
    physicsBodyDefinitions [{
      boundBone {boneName=dyn_shoulderFeathers_01}
      boxExtents=<30 30 30>
      constraintSetup {angularLimitsMax=<10 25 40>, angularLimitsMin=<-10 -25 -40>, angularTarget=<1 0 0>}
      localJointOffset=<5 0 0>
     }]
    relativeSpaceBone {boneName=spine_04}
   }
   n36 = Anim Dynamics - Bone: dyn_shoulderFeathers_02!(
    SimulationSpace:AnimPhysSimSpaceType Enum=Component
    ComponentPose=@n35
    Alpha=1
    ComponentLinearAccScale:Vector="1,1,1"
    ComponentLinearVelScale:Vector="1,1,1"
    ExternalForce="0,0,0"
   ) -> (Pose) {
    angularDampingOverride=1
    angularSpringConstant=17000
    bAngularSpring=true
    bOverrideAngularDamping=true
    bUseGravityOverride=true
    gravityScale=0.35
    alphaBoolBlend {blendOption=Linear}
    boundBone {boneName=dyn_shoulderFeathers_02}
    componentLinearAccScale=<1 1 1>
    componentLinearVelScale=<1 1 1>
    gravityOverride {z=-980}
    physicsBodyDefinitions [{
      boundBone {boneName=dyn_shoulderFeathers_02}
      boxExtents=<30 30 30>
      constraintSetup {angularLimitsMax=<20 40 15>, angularLimitsMin=<-20 -40 -15>, angularTarget=<1 0 0>}
      localJointOffset=<5 0 0>
     }]
    relativeSpaceBone {boneName=spine_04}
   }
   n37 = Anim Dynamics - Bone: dyn_handFeathers!(
    SimulationSpace:AnimPhysSimSpaceType Enum=Component
    ComponentPose=@n36
    Alpha=1
    ComponentLinearAccScale:Vector="1,1,1"
    ComponentLinearVelScale:Vector="1,1,1"
    ExternalForce="0,0,0"
   ) -> (Pose) {
    angularDampingOverride=1
    angularSpringConstant=15000
    bAngularSpring=true
    bOverrideAngularDamping=true
    bUseGravityOverride=true
    gravityScale=0.35
    alphaBoolBlend {blendOption=Linear}
    boundBone {boneName=dyn_handFeathers}
    componentAppliedLinearAccClamp=<5000 5000 5000>
    componentLinearAccScale=<1 1 1>
    componentLinearVelScale=<1 1 1>
    gravityOverride {z=-980}
    physicsBodyDefinitions [{
      boundBone {boneName=dyn_handFeathers}
      boxExtents=<25 25 25>
      constraintSetup {angularLimitsMax=<25 25 30>, angularLimitsMin=<-25 -25 -30>, angularTarget=<1 0 0>}
      localJointOffset=<2 0 0>
     }]
    relativeSpaceBone {boneName=hand_r}
   }
   n38 = RigidBody!(ComponentPose=@n37, Alpha, OverrideWorldGravity:Vector=Get GravityOverride(), ExternalForce) -> (Pose) {
    bOverrideWorldGravity=true
    overridePhysicsAsset=&/F_MED_PlumedStare_Physics
    simulationSpace=BaseBoneSpace
    alphaBoolBlend {blendOption=Linear}
    baseBoneRef {boneName=pelvis}
    componentAppliedLinearAccClamp=<3000 3000 4000>
    componentLinearAccScale=<0.3 0.2 0.4>
    componentLinearVelScale=<0.3 0.2 0.4>
    simSpaceSettings {
     maxAngularAcceleration=1000
     maxAngularVelocity=3
     maxLinearAcceleration=700
     maxLinearVelocity=500
     velocityScaleZ=0
     worldAlpha=0.2
     externalLinearDragV=<0.2 0.2>
    }
   }
   n39 = "Transform (Modify) Bone - Bone: base_hipBowString_01"(
    ComponentPose=@n38
    Alpha=Get isCustomEmotePlaying() -> (isCustomEmotePlaying:double)
    Translation="0,0,-2"
    Rotation="-15,0,0"
    Scale="1,1,1"
   ) {
    rotationMode=BMM_Additive
    rotationSpace=BCS_BoneSpace
    translationMode=BMM_Additive
    translationSpace=BCS_BoneSpace
    alphaBoolBlend {blendOption=Linear}
    alphaScaleBiasClamp {bInterpResult=true, interpSpeedIncreasing=1}
    boneToModify {boneName=base_hipBowString_01}
    rotation {pitch=-15}
    translation {z=-2}
   }
   Save cached pose 'Game AnimDynamics'(Pose=Component To Local(ComponentPose=@n39)) {cacheName=Game AnimDynamics}
   n40 = "Transform (Modify) Bone - Bone: base_skirtBack_01"(
    ComponentPose=Local To Component(LocalPose=Use cached pose 'Surface Constraints'())
    Alpha=1
    Translation="0,-1.5,0"
    Rotation="0,-5,0"
    Scale="1,1,1"
   ) {
    rotationMode=BMM_Additive
    rotationSpace=BCS_BoneSpace
    translationMode=BMM_Additive
    translationSpace=BCS_BoneSpace
    alphaBoolBlend {blendOption=Linear}
    boneToModify {boneName=base_skirtBack_01}
    rotation {yaw=-5}
    translation {y=-1.5}
   }
   n41 = "Transform (Modify) Bone - Bone: base_skirtBackSide_01"(ComponentPose=@n40, Alpha=1, Translation="0,-1.5,0", Rotation="0,-5,0", Scale="1,1,1") {
    rotationMode=BMM_Additive
    rotationSpace=BCS_BoneSpace
    translationMode=BMM_Additive
    translationSpace=BCS_BoneSpace
    alphaBoolBlend {blendOption=Linear}
    boneToModify {boneName=base_skirtBackSide_01}
    rotation {yaw=-5}
    translation {y=-1.5}
   }
   n42 = "Transform (Modify) Bone - Bone: base_skirtSide_01"(ComponentPose=@n41, Alpha=1, Translation="0,-1.5,0", Rotation="0,-5,0", Scale="1,1,1") {
    rotationMode=BMM_Additive
    rotationSpace=BCS_BoneSpace
    translationMode=BMM_Additive
    translationSpace=BCS_BoneSpace
    alphaBoolBlend {blendOption=Linear}
    boneToModify {boneName=base_skirtSide_01}
    rotation {yaw=-5}
    translation {y=-1.5}
   }
   n43 = "Transform (Modify) Bone - Bone: base_skirtSide_01"(ComponentPose=@n42, Alpha=1, Translation="0,-3.5,0", Rotation="0,0,-25", Scale="1,1,1") {
    rotationMode=BMM_Additive
    rotationSpace=BCS_BoneSpace
    translationMode=BMM_Additive
    translationSpace=BCS_BoneSpace
    alphaBoolBlend {blendOption=Linear}
    alphaScaleBiasClamp {bInterpResult=true}
    boneToModify {boneName=base_skirtSide_01}
    rotation {roll=-25}
    translation {y=-3.5}
   }
   n44 = "Transform (Modify) Bone - Bone: dyn_skirtBackSide_02"(ComponentPose=@n43, Alpha=1, Translation="-0.004,0,-0.013", Rotation="0.494,30,-1.436", Scale="1,1,1") {
    rotationMode=BMM_Additive
    rotationSpace=BCS_BoneSpace
    translationMode=BMM_Additive
    translationSpace=BCS_BoneSpace
    alphaBoolBlend {blendOption=Linear}
    alphaScaleBiasClamp {bInterpResult=true}
    boneToModify {boneName=dyn_skirtBackSide_02}
    rotation=<rot 0.494 30 -1.436>
    translation {x=-0.004, z=-0.013}
   }
   n45 = "Transform (Modify) Bone - Bone: base_skirtBack_01"(ComponentPose=@n44, Alpha=1, Translation="0,0,0", Rotation="0.004,-5,0.003", Scale="1,1,1") {
    rotationMode=BMM_Additive
    rotationSpace=BCS_BoneSpace
    translationMode=BMM_Additive
    translationSpace=BCS_BoneSpace
    alphaBoolBlend {blendOption=Linear}
    boneToModify {boneName=base_skirtBack_01}
    rotation=<rot 0.004 -5 0.003>
   }
   n46 = "Transform (Modify) Bone - Bone: dyn_skirtBackSide_04"(ComponentPose=@n45, Alpha=1, Translation="0,0,0", Rotation="11.142,24.980,-5.192", Scale="1,1,1") {
    rotationMode=BMM_Additive
    rotationSpace=BCS_BoneSpace
    translationMode=BMM_Additive
    translationSpace=BCS_BoneSpace
    alphaBoolBlend {blendOption=Linear}
    alphaScaleBiasClamp {bInterpResult=true}
    boneToModify {boneName=dyn_skirtBackSide_04}
    rotation=<rot 11.142 24.98 -5.192>
   }
   n47 = "Transform (Modify) Bone - Bone: dyn_skirtBackSide_03"(ComponentPose=@n46, Alpha=1, Translation="-3.337,4.667,-1.738", Rotation="-0.919,30.988,4.183", Scale="1,1,1") {
    rotationMode=BMM_Additive
    rotationSpace=BCS_BoneSpace
    translationMode=BMM_Additive
    translationSpace=BCS_BoneSpace
    alphaBoolBlend {blendOption=Linear}
    alphaScaleBiasClamp {bInterpResult=true}
    boneToModify {boneName=dyn_skirtBackSide_03}
    rotation=<rot -0.919 30.988 4.183>
    translation=<-3.337 4.667 -1.738>
   }
   n48 = "Transform (Modify) Bone - Bone: dyn_skirtBackSide_02"(ComponentPose=@n47, Alpha=1, Translation="-0.543,4.704,-1.156", Rotation="-0.076,18.245,2.524", Scale="1,1,1") {
    rotationMode=BMM_Additive
    rotationSpace=BCS_BoneSpace
    translationMode=BMM_Additive
    translationSpace=BCS_BoneSpace
    alphaBoolBlend {blendOption=Linear}
    alphaScaleBiasClamp {bInterpResult=true}
    boneToModify {boneName=dyn_skirtBackSide_02}
    rotation=<rot -0.076 18.245 2.524>
    translation=<-0.543 4.704 -1.156>
   }
   n49 = "Transform (Modify) Bone - Bone: base_skirtBack_01"(ComponentPose=@n48, Alpha=1, Translation="0,0,0", Rotation="0.019,9.368,0.006", Scale="1,1,1") {
    rotationMode=BMM_Additive
    rotationSpace=BCS_BoneSpace
    translationMode=BMM_Additive
    translationSpace=BCS_BoneSpace
    alphaBoolBlend {blendOption=Linear}
    alphaScaleBiasClamp {bInterpResult=true}
    boneToModify {boneName=base_skirtBack_01}
    rotation=<rot 0.019 9.368 0.006>
   }
   n50 = "Transform (Modify) Bone - Bone: base_skirtBack_01"(ComponentPose=@n49, Alpha=1, Translation="0,0,0", Rotation="0,-50,0", Scale="1,1,1") {
    rotationMode=BMM_Additive
    rotationSpace=BCS_BoneSpace
    translationMode=BMM_Additive
    translationSpace=BCS_BoneSpace
    alphaBoolBlend {blendOption=Linear}
    alphaScaleBiasClamp {bInterpResult=true, interpSpeedIncreasing=5}
    boneToModify {boneName=base_skirtBack_01}
    rotation {yaw=-50}
   }
   n51 = "Transform (Modify) Bone - Bone: dyn_skirtBack_02"(ComponentPose=@n50, Alpha=1, Translation="0,0,0", Rotation="0,5.503,15.18", Scale="1,1,1") {
    rotationMode=BMM_Additive
    rotationSpace=BCS_BoneSpace
    translationMode=BMM_Additive
    translationSpace=BCS_BoneSpace
    alphaBoolBlend {blendOption=Linear}
    alphaScaleBiasClamp {bInterpResult=true, interpSpeedIncreasing=5}
    boneToModify {boneName=dyn_skirtBack_02}
    rotation {roll=15.18, yaw=5.503}
   }
   n52 = "Transform (Modify) Bone - Bone: base_skirtBackSide_01"(ComponentPose=@n51, Alpha=1, Translation="0,0,0", Rotation="-0.000,-76.541,0.000", Scale="1,1,1") {
    rotationMode=BMM_Additive
    rotationSpace=BCS_BoneSpace
    translationMode=BMM_Additive
    translationSpace=BCS_BoneSpace
    alphaBoolBlend {blendOption=Linear}
    alphaScaleBiasClamp {bInterpResult=true, interpSpeedIncreasing=5}
    boneToModify {boneName=base_skirtBackSide_01}
    rotation {yaw=-76.541}
   }
   n53 = "Transform (Modify) Bone - Bone: dyn_skirtBackSide_02"(ComponentPose=@n52, Alpha=1, Translation="0,0,0", Rotation="8.193,-37.505,22.655", Scale="1,1,1") {
    rotationMode=BMM_Additive
    rotationSpace=BCS_BoneSpace
    translationMode=BMM_Additive
    translationSpace=BCS_BoneSpace
    alphaBoolBlend {blendOption=Linear}
    alphaScaleBiasClamp {bInterpResult=true, interpSpeedIncreasing=5}
    boneToModify {boneName=dyn_skirtBackSide_02}
    rotation=<rot 8.193 -37.5049 22.655>
   }
   n54 = "Transform (Modify) Bone - Bone: dyn_skirtBackSide_03"(ComponentPose=@n53, Alpha=1, Translation="0,0,0", Rotation="-28.782,5.963,11.874", Scale="1,1,1") {
    rotationMode=BMM_Additive
    rotationSpace=BCS_BoneSpace
    translationMode=BMM_Additive
    translationSpace=BCS_BoneSpace
    alphaBoolBlend {blendOption=Linear}
    alphaScaleBiasClamp {bInterpResult=true, interpSpeedIncreasing=5}
    boneToModify {boneName=dyn_skirtBackSide_03}
    rotation=<rot -28.782 5.963 11.874>
   }
   n55 = "Transform (Modify) Bone - Bone: dyn_skirtBackSide_04"(ComponentPose=@n54, Alpha=1, Translation="0,0,0", Rotation="3.390,23.583,12.252", Scale="1,1,1") {
    rotationMode=BMM_Additive
    rotationSpace=BCS_BoneSpace
    translationMode=BMM_Additive
    translationSpace=BCS_BoneSpace
    alphaBoolBlend {blendOption=Linear}
    alphaScaleBiasClamp {bInterpResult=true, interpSpeedIncreasing=5}
    boneToModify {boneName=dyn_skirtBackSide_04}
    rotation=<rot 3.39 23.583 12.252>
   }
   n56 = "Transform (Modify) Bone - Bone: base_skirtSide_01"(ComponentPose=@n55, Alpha=1, Translation="0,0,0", Rotation="0.000,-36.883,-0.000", Scale="1,1,1") {
    rotationMode=BMM_Additive
    rotationSpace=BCS_BoneSpace
    translationMode=BMM_Additive
    translationSpace=BCS_BoneSpace
    alphaBoolBlend {blendOption=Linear}
    alphaScaleBiasClamp {bInterpResult=true, interpSpeedIncreasing=5}
    boneToModify {boneName=base_skirtSide_01}
    rotation=<rot 0 -36.883 0>
   }
   n57 = "Transform (Modify) Bone - Bone: base_skirtBack_01"(ComponentPose=@n56, Alpha=1, Translation="0,0,0", Rotation="0,0,0", Scale="1,1,1") {
    rotationMode=BMM_Additive
    rotationSpace=BCS_BoneSpace
    translationMode=BMM_Additive
    translationSpace=BCS_BoneSpace
    alphaBoolBlend {blendOption=Linear}
    alphaScaleBiasClamp {bInterpResult=true}
    boneToModify {boneName=base_skirtBack_01}
   }
   n58 = "Transform (Modify) Bone - Bone: dyn_skirtBack_02"(ComponentPose=@n57, Alpha=1, Translation="0,0,0", Rotation="0,0,0", Scale="1,1,1") {
    rotationMode=BMM_Additive
    rotationSpace=BCS_BoneSpace
    translationMode=BMM_Additive
    translationSpace=BCS_BoneSpace
    alphaBoolBlend {blendOption=Linear}
    alphaScaleBiasClamp {bInterpResult=true}
    boneToModify {boneName=dyn_skirtBack_02}
   }
   n59 = "Transform (Modify) Bone - Bone: base_skirtBackSide_01"(ComponentPose=@n58, Alpha=1, Translation="0,0,0", Rotation="0.000,-42.997,-0.000", Scale="1,1,1") {
    rotationMode=BMM_Additive
    rotationSpace=BCS_BoneSpace
    translationMode=BMM_Additive
    translationSpace=BCS_BoneSpace
    alphaBoolBlend {blendOption=Linear}
    alphaScaleBiasClamp {bInterpResult=true}
    boneToModify {boneName=base_skirtBackSide_01}
    rotation=<rot 0 -42.997 0>
   }
   n60 = "Transform (Modify) Bone - Bone: dyn_skirtBackSide_02"(ComponentPose=@n59, Alpha=1, Translation="0,0,0", Rotation="-5.425,-17.893,20.790", Scale="1,1,1") {
    rotationMode=BMM_Additive
    rotationSpace=BCS_BoneSpace
    translationMode=BMM_Additive
    translationSpace=BCS_BoneSpace
    alphaBoolBlend {blendOption=Linear}
    alphaScaleBiasClamp {bInterpResult=true}
    boneToModify {boneName=dyn_skirtBackSide_02}
    rotation=<rot -5.425 -17.893 20.7899>
   }
   n61 = "Transform (Modify) Bone - Bone: dyn_skirtBackSide_03"(ComponentPose=@n60, Alpha=1, Translation="0,0,0", Rotation="14.727,-45.080,42.756", Scale="1,1,1") {
    rotationMode=BMM_Additive
    rotationSpace=BCS_BoneSpace
    translationMode=BMM_Additive
    translationSpace=BCS_BoneSpace
    alphaBoolBlend {blendOption=Linear}
    alphaScaleBiasClamp {bInterpResult=true}
    boneToModify {boneName=dyn_skirtBackSide_03}
    rotation=<rot 14.727 -45.08 42.756>
   }
   n62 = "Transform (Modify) Bone - Bone: base_skirtSide_01"(ComponentPose=@n61, Alpha=1, Translation="0,0,0", Rotation="0,0,0", Scale="1,1,1") {
    rotationMode=BMM_Additive
    rotationSpace=BCS_BoneSpace
    translationMode=BMM_Additive
    translationSpace=BCS_BoneSpace
    alphaBoolBlend {blendOption=Linear}
    alphaScaleBiasClamp {bInterpResult=true}
    boneToModify {boneName=base_skirtSide_01}
   }
   Save cached pose 'Correctives'(Pose=Component To Local(ComponentPose=@n62)) {cacheName=Correctives}
  ]}
 EventGraph {nodes [
   Event BlueprintUpdateAnimation()
   > n1 = Sequence()
   > n2 = Gravity Override State Function DEPRECATED(CharacterPartAnimInstance=Self-Reference(), GameTimeInSeconds=GetGameTimeInSeconds(), GravityOverrideParams=Get Gravity Override Params())
   > Set GravityOverride(GravityOverride=@n2)
   > n3 = Sequence()
   n4 = Get OnZipline() -> (OnZipline:bool)
   > n5 = Branch(Condition=@n4)
   > Set Skirt ACC/VEL Zipline(Skirt ACC/VEL Zipline:Vector="0, 0, 0") -> (Output_Get:Vector)
   Set Skirt ACC/VEL Zipline(execute=@n5.else, Skirt ACC/VEL Zipline="0.5,0.5,0.5")
   n6 = Branch(execute=@n3.then_1, Condition=@n4)
   > Set ACC/VEL Zipline(ACC/VEL Zipline:Vector="0, 0, 0") -> (Output_Get:Vector)
   Set ACC/VEL Zipline(execute=@n6.else, ACC/VEL Zipline="1,1,1")
   n7 = GetAnimInstance(self=GetOwnerSkeletalMeshComponent())
   Is Valid(exec=@n1.then_1, InputObject=@n7)
   > Set isCustomEmotePlaying(isCustomEmotePlaying:double=SelectFloat(A=1, B=0, bPickA=Montage_IsPlaying(self=@n7, Montage=/BRCosmetics/Animation/Game/MainPlayer/Emotes/LogarithmKick/Emote_LogarithmKick_CMF_M))) -> (Output_Get:double)
   n8 = Event BlueprintProcessAnimInputEvent()
   > n9 = Branch(Condition=IsNamed(self=@n8.AnimInputEvent, InEventName=aie.Player.Zipline.Entered))
   > Set OnZipline(OnZipline:bool=1) -> (Output_Get:bool)
   Branch(execute=@n9.else, Condition=IsNamed(self=@n8.AnimInputEvent, InEventName=aie.Player.Zipline.Exited))
   > Set OnZipline(OnZipline=false)
   n10 = Event OnCharacterCustomizationLoaded()
   > Set Skel_Mesh(Skel_Mesh=@n10.CustomBaseMesh)
   > n11 = Branch(Condition=Get bIsInFrontEnd())
   > Set Shoulder Sim Space(Shoulder Sim Space:AnimPhysSimSpaceType Enum=World) -> (Output_Get:AnimPhysSimSpaceType Enum)
   > Set Skirt Sim Space(Skirt Sim Space:AnimPhysSimSpaceType Enum=World) -> (Output_Get:AnimPhysSimSpaceType Enum)
   > Set Hand Sim Space(Hand Sim Space:AnimPhysSimSpaceType Enum=Component) -> (Output_Get:AnimPhysSimSpaceType Enum)
   > Set Skirt Alpha(Skirt Alpha:float=0.2) -> (Output_Get:float)
   > Set Shoulder Alpha(Shoulder Alpha:float=1) -> (Output_Get:float)
   > Set Hand Alpha(Hand Alpha:float=1) -> (Output_Get:float)
   Set Shoulder Sim Space(execute=@n11.else, Shoulder Sim Space=Component)
   > Set Skirt Sim Space()
   > Set Hand Sim Space(Hand Sim Space=Component)
   > Set Skirt Alpha(Skirt Alpha=1)
   > Set Shoulder Alpha(Shoulder Alpha=1)
   > Set Hand Alpha(Hand Alpha=1)
  ]}
]
variables [
 ACC/VEL Zipline {type=Vector}
 Gravity Override Params {type=GravityOverrideParamsStruct}
 GravityOverride {type=Vector}
 Hand Alpha {type=number, default=1}
 Hand Sim Space {type=AnimPhysSimSpaceType}
 isCustomEmotePlaying {type=number}
 OnZipline {type=boolean}
 Shoulder Alpha {type=number, default=1}
 Shoulder Sim Space {type=AnimPhysSimSpaceType}
 Sim Space Feathers {type=boolean}
 Skel_Mesh {type=SkeletalMeshComponent}
 Skirt ACC/VEL Zipline {type=Vector}
 Skirt Alpha {type=number, default=1}
 Skirt Sim Space {type=AnimPhysSimSpaceType, default=RootRelative}
]
