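# Physics asset of collision capsules for the PlumedStare medium female player body, named as colliders for its cape.
# Capsules cover the pelvis, spine, clavicles, head and feet, plus skirt, feather and right-thigh scroll bones, while arm, thigh, calf and neck bodies have no shapes.
# Constraints follow the bone chains, with feathers branching from the back-side skirt, and leave swing and twist free.
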PhysicsAsset /BRCosmetics/Characters/Player/Female/Medium/Bodies/F_MED_PlumedStare/Meshes/F_MED_PlumedStare_CapeColliders
bodies {
 calf_l []
 calf_r []
 clavicle_l [{shapeType=Capsule, length=6.5623, radius=7.5194, center=<3.9566 -1.0862 -2.1127>, rotation=<rot -77.5566 177.3915 7.7314>}]
 clavicle_r [{shapeType=Capsule, length=6.5623, radius=7.5194, center=<-3.9566 1.0862 2.1127>, rotation=<rot -77.5566 177.3915 -172.2686>}]
 dyn_skirtBackSide_01 [{
   shapeType=Capsule
   length=13.2202
   radius=8.9675
   shapeName=dyn_skirtBackSide_01_capsule
   center=<5.0262 4.3222 0.2279>
   rotation=<rot 0 0 -9.8048>
  }]
 dyn_skirtBackSide_02 [{
   shapeType=Capsule
   length=14.5658
   radius=11.1316
   shapeName=dyn_skirtBackSide_02_capsule
   center=<5.915 5.1366 0.3155>
   rotation=<rot 0 0 -11.5704>
  }]
 dyn_skirtBackSide_03 [{
   shapeType=Capsule
   length=20.9029
   radius=9.5719
   shapeName=dyn_skirtBackSide_03_capsule
   center=<4.4417 4.1325 0.3727>
   rotation=<rot 0.0002 0.0002 -10.7645>
  }]
 dyn_skirtBackSide_04 [{
   shapeType=Capsule
   length=23.1813
   radius=10.7906
   shapeName=dyn_skirtBackSide_04_capsule
   center=<3.2703 3.518 -2.3275>
   rotation=<rot 0 0.0001 -10.4045>
  }]
 dyn_skirtBack_01 [{shapeType=Capsule, length=6.7578, radius=5.8191, shapeName=dyn_skirtBack_01_capsule, center=<5.4874 3.5506 0>}]
 dyn_skirtBack_02 [{
   shapeType=Capsule
   length=5.9442
   radius=4.8332
   shapeName=dyn_skirtBack_02_capsule
   center=<4.019 4.2307 0>
   rotation=<rot 90 -3.5763 -13.107>
  }]
 dyn_skirtFeather_00_01 [{shapeType=Capsule, length=14.2549, radius=9.4057, shapeName=dyn_skirtFeather_00_01_capsule, center=<2.0622 0.3476 0>, rotation=<rot 90 0 0>}]
 dyn_skirtFeather_01_01 [{
   shapeType=Capsule
   length=14.8555
   radius=11.7163
   shapeName=dyn_skirtFeather_01_01_capsule
   center=<6.3179 0 -0.3956>
   rotation=<rot 82.7681 0.0001 0.0001>
  }]
 dyn_skirtFeather_01_02 [{
   shapeType=Capsule
   length=14.096
   radius=11.183
   shapeName=dyn_skirtFeather_01_02_capsule
   center=<5.7332 0 0>
   rotation=<rot 82.9165 0.0005 0.0004>
  }]
 dyn_skirtFeather_01_03 [{shapeType=Capsule, length=13.0676, radius=11.0289, shapeName=dyn_skirtFeather_01_03_capsule, center=<1.3836 0 0>, rotation=<rot 90 0 0>}]
 dyn_skirtFeather_02_01 [{shapeType=Capsule, length=14.1424, radius=10.2069, shapeName=dyn_skirtFeather_02_01_capsule, center=<4.3541 0 0>, rotation=<rot 90 0 0>}]
 dyn_skirtFeather_02_02 [{shapeType=Capsule, length=12.4764, radius=10.5867, shapeName=dyn_skirtFeather_02_02_capsule, center=<3.2763 0 0>, rotation=<rot 90 0 0>}]
 dyn_skirtFeather_02_03 [{shapeType=Capsule, length=13.0547, radius=9.9772, shapeName=dyn_skirtFeather_02_03_capsule, center=<0.1681 -0.3762 0>, rotation=<rot 90 0 0>}]
 dyn_skirtFeather_03_01 [{shapeType=Capsule, length=8.8947, radius=9.6761, shapeName=dyn_skirtFeather_03_01_capsule, center=<2 0 0>, rotation=<rot 90 0 0>}]
 dyn_skirtFeather_03_02 [{shapeType=Capsule, length=9.5962, radius=9.8205, shapeName=dyn_skirtFeather_03_02_capsule, center=<2 0 0>, rotation=<rot 90 0 0>}]
 dyn_skirtFeather_03_03 [{shapeType=Capsule, length=11.8097, radius=9.4728, shapeName=dyn_skirtFeather_03_03_capsule, center=<1.1004 -0.5306 0>, rotation=<rot 90 0 0>}]
 dyn_skirtFeather_04_01 [{shapeType=Capsule, length=8.5057, radius=10.8183, shapeName=dyn_skirtFeather_04_01_capsule, center=<2 0 0>, rotation=<rot 90 0 0>}]
 dyn_skirtFeather_04_02 [{
   shapeType=Capsule
   length=9.1974
   radius=11.2142
   shapeName=dyn_skirtFeather_04_02_capsule
   center=<1.5645 -1.1108 0>
   rotation=<rot 90 180 -146.3231>
  }]
 dyn_skirtSide_01 [{shapeType=Capsule, length=5.5557, radius=6.0674, shapeName=dyn_skirtSide_01_capsule, center=<4.5158 3.5601 0>}]
 dyn_skirtSide_02 [{
   shapeType=Capsule
   length=6.2655
   radius=5.7148
   shapeName=dyn_skirtSide_02_capsule
   center=<4.7968 3.4803 -0.0517>
   rotation=<rot 0.0001 0 11.4292>
  }]
 dyn_skirtSide_03 [{
   shapeType=Capsule
   length=5.8514
   radius=5.491
   shapeName=dyn_skirtSide_03_capsule
   center=<4.7824 -3.2379 3.1723>
   rotation=<rot 74.8313 87.2892 -179.9999>
  }]
 foot_l [{shapeType=Capsule, length=11.0765, radius=8.0344, center=<-4.1811 6.0027 -0.338>, rotation=<rot 0.0388 1.4701 -89.3982>}]
 foot_r [{shapeType=Capsule, length=11.0765, radius=8.0344, center=<4.1811 -6.0027 0.338>, rotation=<rot 0.0388 1.4701 90.6018>}]
 head [{shapeType=Capsule, length=0, radius=18, center=<9.0697 0.0746 0>, rotation=<rot 0 -90 0>}]
 legScrollString_01 [{shapeType=Capsule, length=3.277, radius=1.2808, shapeName=legScrollString_01_capsule, center=<0.717 0 0>, rotation=<rot 90 0 0>}]
 legScrollString_02 [{shapeType=Capsule, length=3.1137, radius=1.5912, shapeName=legScrollString_02_capsule, center=<2.736 0 0>, rotation=<rot 90 0 0>}]
 legScroll_01 [{
   shapeType=Capsule
   length=11.2891
   radius=3.2562
   shapeName=legScroll_01_capsule
   center=<-0.0828 -0.142 -0.0186>
   rotation=<rot -33.405 -90 -3.3838>
  }]
 legScroll_02 [{
   shapeType=Capsule
   length=11.2891
   radius=3.2562
   shapeName=legScroll_02_capsule
   center=<0.2775 0.0067 -0.0163>
   rotation=<rot -31.3802 -90.6742 6.2972>
  }]
 lowerarm_l []
 lowerarm_r []
 neck_02 []
 pelvis [{shapeType=Capsule, length=13.3486, radius=12.7038, center=<-2.4471 -2.5289 -0.1981>}]
 spine_01 [{shapeType=Capsule, length=4.3888, radius=14.3638, center=<-8.9021 -0.8973 0>}]
 spine_02 [
  {shapeType=Capsule, length=2.5219, radius=9.7317, center=<-2.6885 -1.2641 0>}
  {shapeType=Capsule, length=20.6875, radius=9.3125, center=<-1.8317 1.9362 0>, rotation=<rot 90 0 0>}
 ]
 spine_03 [
  {shapeType=Capsule, length=6.1863, radius=8.4854, center=<0.0574 0.4416 0>}
  {shapeType=Capsule, length=20.7188, radius=9.2812, center=<-1.2163 0.549 0>, rotation=<rot 90 180 -175.0732>}
 ]
 spine_04 [
  {shapeType=Capsule, length=9.0076, radius=10.3916, center=<5.6234 -0.6779 0>}
  {shapeType=Capsule, length=10.0312, radius=9.2812, center=<-1.2326 1.308 0>, rotation=<rot 90 0 0>}
 ]
 spine_05 [{shapeType=Capsule, length=9.678, radius=12.3571, center=<0.6969 1.91 0>}]
 thigh_l []
 thigh_r []
 upperarm_l []
 upperarm_r []
}
constraints [
 {
  bone1Name=spine_01
  bone2Name=pelvis
  swing1LimitDegrees=45
  swing1Motion=Free
  swing2LimitDegrees=45
  swing2Motion=Free
  twistLimitDegrees=45
  twistMotion=Free
 }
 {
  bone1Name=spine_02
  bone2Name=spine_01
  swing1LimitDegrees=45
  swing1Motion=Free
  swing2LimitDegrees=45
  swing2Motion=Free
  twistLimitDegrees=45
  twistMotion=Free
 }
 {
  bone1Name=spine_03
  bone2Name=spine_02
  swing1LimitDegrees=45
  swing1Motion=Free
  swing2LimitDegrees=45
  swing2Motion=Free
  twistLimitDegrees=45
  twistMotion=Free
 }
 {
  bone1Name=spine_04
  bone2Name=spine_03
  swing1LimitDegrees=45
  swing1Motion=Free
  swing2LimitDegrees=45
  swing2Motion=Free
  twistLimitDegrees=45
  twistMotion=Free
 }
 {
  bone1Name=spine_05
  bone2Name=spine_04
  swing1LimitDegrees=45
  swing1Motion=Free
  swing2LimitDegrees=45
  swing2Motion=Free
  twistLimitDegrees=45
  twistMotion=Free
 }
 {
  bone1Name=clavicle_l
  bone2Name=spine_05
  swing1LimitDegrees=45
  swing1Motion=Free
  swing2LimitDegrees=45
  swing2Motion=Free
  twistLimitDegrees=45
  twistMotion=Free
 }
 {
  bone1Name=upperarm_l
  bone2Name=clavicle_l
  swing1LimitDegrees=45
  swing1Motion=Free
  swing2LimitDegrees=45
  swing2Motion=Free
  twistLimitDegrees=45
  twistMotion=Free
 }
 {
  bone1Name=lowerarm_l
  bone2Name=upperarm_l
  swing1LimitDegrees=45
  swing1Motion=Free
  swing2LimitDegrees=45
  swing2Motion=Free
  twistLimitDegrees=45
  twistMotion=Free
 }
 {
  bone1Name=neck_02
  bone2Name=spine_05
  swing1LimitDegrees=45
  swing1Motion=Free
  swing2LimitDegrees=45
  swing2Motion=Free
  twistLimitDegrees=45
  twistMotion=Free
 }
 {
  bone1Name=head
  bone2Name=neck_02
  swing1LimitDegrees=45
  swing1Motion=Free
  swing2LimitDegrees=45
  swing2Motion=Free
  twistLimitDegrees=45
  twistMotion=Free
 }
 {
  bone1Name=thigh_l
  bone2Name=pelvis
  swing1LimitDegrees=45
  swing1Motion=Free
  swing2LimitDegrees=45
  swing2Motion=Free
  twistLimitDegrees=45
  twistMotion=Free
 }
 {
  bone1Name=foot_l
  bone2Name=calf_l
  swing1LimitDegrees=45
  swing1Motion=Free
  swing2LimitDegrees=45
  swing2Motion=Free
  twistLimitDegrees=45
  twistMotion=Free
 }
 {
  bone1Name=lowerarm_r
  bone2Name=upperarm_r
  swing1LimitDegrees=45
  swing1Motion=Free
  swing2LimitDegrees=45
  swing2Motion=Free
  twistLimitDegrees=45
  twistMotion=Free
 }
 {
  bone1Name=upperarm_r
  bone2Name=clavicle_r
  swing1LimitDegrees=45
  swing1Motion=Free
  swing2LimitDegrees=45
  swing2Motion=Free
  twistLimitDegrees=45
  twistMotion=Free
 }
 {
  bone1Name=clavicle_r
  bone2Name=spine_05
  swing1LimitDegrees=45
  swing1Motion=Free
  swing2LimitDegrees=45
  swing2Motion=Free
  twistLimitDegrees=45
  twistMotion=Free
 }
 {
  bone1Name=thigh_r
  bone2Name=pelvis
  swing1LimitDegrees=45
  swing1Motion=Free
  swing2LimitDegrees=45
  swing2Motion=Free
  twistLimitDegrees=45
  twistMotion=Free
 }
 {
  bone1Name=foot_r
  bone2Name=calf_r
  swing1LimitDegrees=45
  swing1Motion=Free
  swing2LimitDegrees=45
  swing2Motion=Free
  twistLimitDegrees=45
  twistMotion=Free
 }
 {
  bone1Name=calf_r
  bone2Name=thigh_r
  swing1LimitDegrees=45
  swing1Motion=Free
  swing2LimitDegrees=45
  swing2Motion=Free
  twistLimitDegrees=45
  twistMotion=Free
 }
 {
  bone1Name=calf_l
  bone2Name=thigh_l
  swing1LimitDegrees=45
  swing1Motion=Free
  swing2LimitDegrees=45
  swing2Motion=Free
  twistLimitDegrees=45
  twistMotion=Free
 }
 {
  bone1Name=dyn_skirtBackSide_01
  bone2Name=pelvis
  swing1LimitDegrees=45
  swing1Motion=Free
  swing2LimitDegrees=45
  swing2Motion=Free
  twistLimitDegrees=45
  twistMotion=Free
 }
 {
  bone1Name=dyn_skirtBackSide_02
  bone2Name=dyn_skirtBackSide_01
  swing1LimitDegrees=45
  swing1Motion=Free
  swing2LimitDegrees=45
  swing2Motion=Free
  twistLimitDegrees=45
  twistMotion=Free
 }
 {
  bone1Name=dyn_skirtBackSide_03
  bone2Name=dyn_skirtBackSide_02
  swing1LimitDegrees=45
  swing1Motion=Free
  swing2LimitDegrees=45
  swing2Motion=Free
  twistLimitDegrees=45
  twistMotion=Free
 }
 {
  bone1Name=dyn_skirtBackSide_04
  bone2Name=dyn_skirtBackSide_03
  swing1LimitDegrees=45
  swing1Motion=Free
  swing2LimitDegrees=45
  swing2Motion=Free
  twistLimitDegrees=45
  twistMotion=Free
 }
 {
  bone1Name=dyn_skirtFeather_01_02
  bone2Name=dyn_skirtFeather_01_01
  swing1LimitDegrees=45
  swing1Motion=Free
  swing2LimitDegrees=45
  swing2Motion=Free
  twistLimitDegrees=45
  twistMotion=Free
 }
 {
  bone1Name=dyn_skirtFeather_01_01
  bone2Name=dyn_skirtBackSide_04
  swing1LimitDegrees=45
  swing1Motion=Free
  swing2LimitDegrees=45
  swing2Motion=Free
  twistLimitDegrees=45
  twistMotion=Free
 }
 {
  bone1Name=dyn_skirtFeather_01_03
  bone2Name=dyn_skirtFeather_01_02
  swing1LimitDegrees=45
  swing1Motion=Free
  swing2LimitDegrees=45
  swing2Motion=Free
  twistLimitDegrees=45
  twistMotion=Free
 }
 {
  bone1Name=dyn_skirtFeather_02_01
  bone2Name=dyn_skirtBackSide_04
  swing1LimitDegrees=45
  swing1Motion=Free
  swing2LimitDegrees=45
  swing2Motion=Free
  twistLimitDegrees=45
  twistMotion=Free
 }
 {
  bone1Name=dyn_skirtFeather_02_02
  bone2Name=dyn_skirtFeather_02_01
  swing1LimitDegrees=45
  swing1Motion=Free
  swing2LimitDegrees=45
  swing2Motion=Free
  twistLimitDegrees=45
  twistMotion=Free
 }
 {
  bone1Name=dyn_skirtFeather_02_03
  bone2Name=dyn_skirtFeather_02_02
  swing1LimitDegrees=45
  swing1Motion=Free
  swing2LimitDegrees=45
  swing2Motion=Free
  twistLimitDegrees=45
  twistMotion=Free
 }
 {
  bone1Name=dyn_skirtFeather_03_01
  bone2Name=dyn_skirtBackSide_04
  swing1LimitDegrees=45
  swing1Motion=Free
  swing2LimitDegrees=45
  swing2Motion=Free
  twistLimitDegrees=45
  twistMotion=Free
 }
 {
  bone1Name=dyn_skirtFeather_03_02
  bone2Name=dyn_skirtFeather_03_01
  swing1LimitDegrees=45
  swing1Motion=Free
  swing2LimitDegrees=45
  swing2Motion=Free
  twistLimitDegrees=45
  twistMotion=Free
 }
 {
  bone1Name=dyn_skirtFeather_03_03
  bone2Name=dyn_skirtFeather_03_02
  swing1LimitDegrees=45
  swing1Motion=Free
  swing2LimitDegrees=45
  swing2Motion=Free
  twistLimitDegrees=45
  twistMotion=Free
 }
 {
  bone1Name=dyn_skirtFeather_04_01
  bone2Name=dyn_skirtBackSide_04
  swing1LimitDegrees=45
  swing1Motion=Free
  swing2LimitDegrees=45
  swing2Motion=Free
  twistLimitDegrees=45
  twistMotion=Free
 }
 {
  bone1Name=dyn_skirtFeather_04_02
  bone2Name=dyn_skirtFeather_04_01
  swing1LimitDegrees=45
  swing1Motion=Free
  swing2LimitDegrees=45
  swing2Motion=Free
  twistLimitDegrees=45
  twistMotion=Free
 }
 {
  bone1Name=dyn_skirtFeather_00_01
  bone2Name=dyn_skirtBackSide_03
  swing1LimitDegrees=45
  swing1Motion=Free
  swing2LimitDegrees=45
  swing2Motion=Free
  twistLimitDegrees=45
  twistMotion=Free
 }
 {
  bone1Name=dyn_skirtBack_01
  bone2Name=pelvis
  swing1LimitDegrees=45
  swing1Motion=Free
  swing2LimitDegrees=45
  swing2Motion=Free
  twistLimitDegrees=45
  twistMotion=Free
 }
 {
  bone1Name=dyn_skirtBack_02
  bone2Name=dyn_skirtBack_01
  swing1LimitDegrees=45
  swing1Motion=Free
  swing2LimitDegrees=45
  swing2Motion=Free
  twistLimitDegrees=45
  twistMotion=Free
 }
 {
  bone1Name=dyn_skirtSide_01
  bone2Name=pelvis
  swing1LimitDegrees=45
  swing1Motion=Free
  swing2LimitDegrees=45
  swing2Motion=Free
  twistLimitDegrees=45
  twistMotion=Free
 }
 {
  bone1Name=dyn_skirtSide_02
  bone2Name=dyn_skirtSide_01
  swing1LimitDegrees=45
  swing1Motion=Free
  swing2LimitDegrees=45
  swing2Motion=Free
  twistLimitDegrees=45
  twistMotion=Free
 }
 {
  bone1Name=dyn_skirtSide_03
  bone2Name=dyn_skirtSide_02
  swing1LimitDegrees=45
  swing1Motion=Free
  swing2LimitDegrees=45
  swing2Motion=Free
  twistLimitDegrees=45
  twistMotion=Free
 }
 {
  bone1Name=legScroll_01
  bone2Name=thigh_r
  swing1LimitDegrees=45
  swing1Motion=Free
  swing2LimitDegrees=45
  swing2Motion=Free
  twistLimitDegrees=45
  twistMotion=Free
 }
 {
  bone1Name=legScroll_02
  bone2Name=thigh_r
  swing1LimitDegrees=45
  swing1Motion=Free
  swing2LimitDegrees=45
  swing2Motion=Free
  twistLimitDegrees=45
  twistMotion=Free
 }
 {
  bone1Name=legScrollString_01
  bone2Name=legScroll_02
  swing1LimitDegrees=45
  swing1Motion=Free
  swing2LimitDegrees=45
  swing2Motion=Free
  twistLimitDegrees=45
  twistMotion=Free
 }
 {
  bone1Name=legScrollString_02
  bone2Name=legScrollString_01
  swing1LimitDegrees=45
  swing1Motion=Free
  swing2LimitDegrees=45
  swing2Motion=Free
  twistLimitDegrees=45
  twistMotion=Free
 }
]
