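# Post-process-style AnimBP for the PlumedStare face accessory, adding hair physics and variant-based element hiding.
# Copies the attached parent's pose, then chains four world-space, spring-limited Anim Dynamics bones (dyn_hair_01–04) using per-update gravity override.
# On customization load, the Parts variant index toggles "Hide Element 01" on the dynamic material.
# It also sets front-end-dependent hair alpha and sim space, which the AnimGraph never reads.
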AnimBlueprint /BRCosmetics/Characters/Player/Female/Medium/Bodies/F_MED_PlumedStare/Meshes/Parts/F_MED_PlumedStare_FaceAcc_AnimBP
target_skeleton=&/F_MED_PlumedStare_FaceAcc_Skeleton
graphs [
 AnimGraph {nodes [
   n1 = Get GravityOverride()
   n2 = Anim Dynamics - Bone: dyn_hair_01!(
    SimulationSpace:AnimPhysSimSpaceType Enum=World
    ComponentPose=Local To Component(LocalPose=Copy Pose From Mesh() {bUseAttachedParent=true})
    Alpha=1
    GravityOverride:Vector=@n1
    ExternalForce
   ) -> (Pose) {
    angularDampingOverride=1
    angularSpringConstant=20000
    bAngularSpring=true
    bOverrideAngularDamping=true
    bUseGravityOverride=true
    gravityScale=0.35
    simulationSpace=World
    alphaBoolBlend {blendOption=Linear}
    boundBone {boneName=dyn_hair_01}
    componentAppliedLinearAccClamp=<1000 1000 1000>
    componentLinearAccScale=<1 1 1>
    componentLinearVelScale=<1 1 1>
    gravityOverride {z=-980}
    physicsBodyDefinitions [{
      boundBone {boneName=dyn_hair_01}
      boxExtents=<30 30 30>
      constraintSetup {angularLimitsMax=<0 15 7>, angularLimitsMin=<0 -5 -7>, angularTarget=<1 0 0>}
      localJointOffset=<3 0 0>
     }]
    relativeSpaceBone {boneName=head}
   }
   n3 = Anim Dynamics - Bone: dyn_hair_02!(SimulationSpace:AnimPhysSimSpaceType Enum=World, ComponentPose=@n2, Alpha=1, GravityOverride:Vector=@n1, ExternalForce) -> (Pose) {
    angularDampingOverride=1
    angularSpringConstant=20000
    bAngularSpring=true
    bOverrideAngularDamping=true
    bUseGravityOverride=true
    gravityScale=0.35
    simulationSpace=World
    alphaBoolBlend {blendOption=Linear}
    boundBone {boneName=dyn_hair_02}
    componentAppliedLinearAccClamp=<100 100 100>
    componentLinearAccScale=<1 1 1>
    componentLinearVelScale=<1 1 1>
    gravityOverride {z=-980}
    physicsBodyDefinitions [{
      boundBone {boneName=dyn_hair_02}
      boxExtents=<30 30 30>
      constraintSetup {angularLimitsMax=<0 15 10>, angularLimitsMin=<0 -10 -10>, angularTarget=<1 0 0>}
      localJointOffset=<4 0 0>
     }]
    relativeSpaceBone {boneName=head}
   }
   n4 = Anim Dynamics - Bone: dyn_hair_03!(SimulationSpace:AnimPhysSimSpaceType Enum=World, ComponentPose=@n3, Alpha=1, GravityOverride:Vector=@n1, ExternalForce="0,0,0") -> (Pose) {
    angularDampingOverride=1
    angularSpringConstant=20000
    bAngularSpring=true
    bOverrideAngularDamping=true
    bUseGravityOverride=true
    gravityScale=0.35
    simulationSpace=World
    alphaBoolBlend {blendOption=Linear}
    boundBone {boneName=dyn_hair_03}
    componentAppliedLinearAccClamp=<100 100 100>
    componentLinearAccScale=<1 1 1>
    componentLinearVelScale=<1 1 1>
    gravityOverride {z=-980}
    physicsBodyDefinitions [{
      boundBone {boneName=dyn_hair_03}
      boxExtents=<10 10 10>
      constraintSetup {angularLimitsMax=<0 15 7>, angularLimitsMin=<0 -25 -7>, angularTarget=<1 0 0>}
      localJointOffset=<1 0 0>
     }]
    relativeSpaceBone {boneName=head}
   }
   n5 = Anim Dynamics - Bone: dyn_hair_04!(SimulationSpace:AnimPhysSimSpaceType Enum=World, ComponentPose=@n4, Alpha=1, GravityOverride:Vector=@n1, ExternalForce="0,0,0") -> (Pose) {
    angularDampingOverride=1
    angularSpringConstant=5000
    bAngularSpring=true
    bOverrideAngularDamping=true
    bUseGravityOverride=true
    gravityScale=0.35
    simulationSpace=World
    alphaBoolBlend {blendOption=Linear}
    boundBone {boneName=dyn_hair_04}
    componentAppliedLinearAccClamp=<100 100 100>
    componentLinearAccScale=<1 1 1>
    componentLinearVelScale=<1 1 1>
    gravityOverride {z=-980}
    physicsBodyDefinitions [{
      boundBone {boneName=dyn_hair_04}
      boxExtents=<20 20 20>
      constraintSetup {angularLimitsMax=<0 20 15>, angularLimitsMin=<0 -20 -15>, angularTarget=<1 0 0>}
      localJointOffset=<3 0 0>
     }]
    relativeSpaceBone {boneName=head}
   }
   Output Pose(Result=Component To Local(ComponentPose=@n5)) {layerGroup=None}
  ]}
 EventGraph {nodes [
   Event BlueprintUpdateAnimation()
   > n1 = Gravity Override State Function DEPRECATED(CharacterPartAnimInstance=Self-Reference(), GameTimeInSeconds=GetGameTimeInSeconds(), GravityOverrideParams=Get Gravity Override Params())
   > Set GravityOverride(GravityOverride=@n1)
   n2 = Event OnCharacterCustomizationLoaded()
   > n3 = Sequence()
   > n4 = Cast To FortPlayerPawn(Object=GetOwningActor())
   > n5 = GetVariantChannelActiveIndexForItemActor!(
    execute
    Item:Athena Cosmetic Item Definition Object Reference=/BRCosmetics/Athena/Items/Cosmetics/Characters/Character_PlumedStare
    Actor:Actor Object Reference=@n4
    Channel:Gameplay Tag Structure=(TagName="Cosmetics.Variant.Channel.Parts")
    VariantList:Array of Gameplay Tag Structures=Make Array!("[0]"=(TagName="Cosmetics.Variant.Property.Stage1"), "[1]":Gameplay Tag Structure=(TagName="Cosmetics.Variant.Property.Stage2")) -> (Array)
    bErrorIfChannelMissing:bool=false
    bErrorIfTagNotInArray:bool=false
   ) -> (then, Index:int=0, bIndexIsValid:bool, bFoundChannel:bool)
   > n6 = Get Dynamic Material Instance(SkeletalMeshComponent=GetPartSkeletalMeshComponent())
   > n7 = Branch(Condition="Equal (Integer)"(A=@n5))
   > SetScalarParameterValue(self=@n6, ParameterName=Hide Element 01, Value=1)
   SetScalarParameterValue(execute=@n7.else, self=@n6, ParameterName=Hide Element 01, Value=0)
   Set Skel_Mesh(execute=@n3.then_1, Skel_Mesh=@n2.CustomBaseMesh)
   > n8 = Branch(Condition=Get bIsInFrontEnd())
   > Set Hair Sim Space(Hair Sim Space:AnimPhysSimSpaceType Enum=World) -> (Output_Get:AnimPhysSimSpaceType Enum)
   > Set Hair Alpha(Hair Alpha:float=0.4) -> (Output_Get:float)
   Set Hair Sim Space(execute=@n8.else)
   > Set Hair Alpha(Hair Alpha=1)
  ]}
]
variables [
 Gravity Override Params {type=GravityOverrideParamsStruct}
 GravityOverride {type=Vector}
 Hair Alpha {type=number}
 Hair Sim Space {type=AnimPhysSimSpaceType}
 Skel_Mesh {type=SkeletalMeshComponent}
]
